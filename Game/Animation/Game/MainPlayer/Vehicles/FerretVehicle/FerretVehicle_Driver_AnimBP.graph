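# Anim Blueprint that poses a player driving the Ferret vehicle through the VehicleFinalPoseOverride layer.
# A Locomotion state machine (air, ground, boost, roll) feeds a driver aim offset.
# Additives layer on idle and collision reactions, a gender correction, a JethroXV idle pose enabled by the parent actor's gameplay tag, and shooting.
# After the FullBody slot, curve-driven ik_hand bone offsets feed Leg IK nodes repurposed to solve both arms.
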AnimBlueprint /Game/Animation/Game/MainPlayer/Vehicles/FerretVehicle/FerretVehicle_Driver_AnimBP
target_skeleton=&Fortnite_M_Avg_Player_Skeleton
graphs [
 AnimGraph {nodes [Output Pose() {layerGroup=None}]}
 EventGraph {nodes [
   Event BlueprintUpdateAnimation()
   > n1 = HasMatchingGameplayTag!(
    execute:Exec
    self:Object Reference=GetAttachParentActor!(self:Actor Object Reference=TryGetPawnOwner()) -> (ReturnValue:Actor Object Reference)
    TagToCheck:Gameplay Tag Structure=(TagName="Athena.Vehicle.JethroXV")
   ) -> (then:Exec, ReturnValue:bool) {is=title K2Node_Message}
   > Set JethroXV_Driver_Alpha(JethroXV_Driver_Alpha:double="To Float (Boolean)"(InBool=@n1)) -> (Output_Get:double)
   Event BlueprintInitializeAnimation()
   > Set RightHandLocalRotateOffset(RightHandLocalRotateOffset:Rotator=Get HandLocalRotationOffsetRight() -> (HandLocalRotationOffsetRight:Rotator)) -> (Output_Get:Rotator)
   > Set RightHandLocalTranslateOffset(RightHandLocalTranslateOffset:Vector=Get HandLocalTranslateOffsetRight() -> (HandLocalTranslateOffsetRight:Vector)) -> (Output_Get:Vector)
   > Set LeftHandLocalRotateOffset(LeftHandLocalRotateOffset:Rotator=Get HandLocalRotationOffsetLeft() -> (HandLocalRotationOffsetLeft:Rotator)) -> (Output_Get:Rotator)
   > Set LeftHandLocalTranslateOffset(LeftHandLocalTranslateOffset:Vector=Get HandLocalTranslateOffsetLeft() -> (HandLocalTranslateOffsetLeft:Vector)) -> (Output_Get:Vector)
  ]}
 VehicleFinalPoseOverride {nodes [
   Input Pose() {name=InFinalPose}
   AnimGraphNode_StateMachine_0 = Locomotion() {is=title AnimGraphNode_StateMachine}
   n1 = AimOffset Player 'FerretVehicle_Driver_AO'!(
    X:float=Get DriverAimYaw() -> (DriverAimYaw:float)
    Y:float=Get DriverAimPitch() -> (DriverAimPitch:float)
    BasePose:Pose Link Structure=@AnimGraphNode_StateMachine_0
    Alpha:float=1
   ) -> (Pose:Pose Link Structure) {blendSpace=/Game/Animation/Game/MainPlayer/Vehicles/FerretVehicle/AO/FerretVehicle_Driver_AO, alphaBoolBlend {blendOption=Linear}}
   AnimGraphNode_StateMachine_1 = FullBodyAdditive() {is=title AnimGraphNode_StateMachine}
   n2 = Apply Additive(Base=@n1, Additive=@AnimGraphNode_StateMachine_1) {alphaBoolBlend {blendOption=Linear}}
   n3 = "Blend Poses (EFortCustomGender)"!(
    BlendPose_0:Pose Link Structure=Additive Identity Pose()
    BlendPose_1:Pose Link Structure="Sequence Player (additive) 'FerretVehicle_Idle_Pose_CMF'"() {sequence=&/FerretVehicle_Idle_Pose_CMF}
    BlendTime_0:float=0.1
    BlendTime_1:float=0.1
    ActiveEnumValue:Byte=Get Gender() -> (Gender:EFortCustomGender Enum=Invalid)
   ) -> (Pose:Pose Link Structure) {
    is=title AnimGraphNode_BlendListByEnum
    blendType=Linear
    blendPose [{linkId=-1, sourceLinkId=-1}, {linkId=-1, sourceLinkId=-1}]
    blendTime [0.1,0.1]
   }
   AnimGraphNode_StateMachine_2 = Shooting() {is=title AnimGraphNode_StateMachine}
   n4 = Apply Additive(
    Base=Apply Additive(
     Base=Apply Additive(Base=@n2, Additive=@n3) {alphaBoolBlend {blendOption=Linear}}
     Additive="Sequence Player (additive) 'JethroXV_Driver_IdlePose'"() {bLoopAnimation=false, sequence=&/JethroXV_Driver_IdlePose}
     Alpha=Get JethroXV_Driver_Alpha() -> (JethroXV_Driver_Alpha:double)
    ) {alphaBoolBlend {blendOption=Linear}}
    Additive=@AnimGraphNode_StateMachine_2
   ) {alphaBoolBlend {blendOutTime=0.2}}
   n5 = "Transform (Modify) Bone - Bone: ik_hand_r"!(
    ComponentPose=Local To Component(LocalPose=Slot 'FullBody'(Source=@n4) {slotName=FullBody})
    AlphaCurveName:Name=RightHandIK
    Translation=Get RightHandComponentTranslateOffset() -> (RightHandComponentTranslateOffset:Vector)
    Rotation=Get RightHandComponentRotateOffset() -> (RightHandComponentRotateOffset:Rotator)
   ) -> (Pose) {
    alphaInputType=Curve
    rotationMode=BMM_Replace
    translationMode=BMM_Replace
    alphaBoolBlend {blendOption=Linear}
    alphaScaleBiasClamp {bias=1, scale=-1}
    boneToModify {boneName=ik_hand_r}
   }
   n6 = "Transform (Modify) Bone - Bone: ik_hand_l"!(
    ComponentPose=@n5
    AlphaCurveName:Name=LeftHandIK
    Translation=Get LeftHandComponentTranslateOffset() -> (LeftHandComponentTranslateOffset:Vector)
    Rotation=Get LeftHandComponentRotateOffset() -> (LeftHandComponentRotateOffset:Rotator)
   ) -> (Pose) {
    alphaInputType=Curve
    rotationMode=BMM_Replace
    translationMode=BMM_Replace
    alphaBoolBlend {blendOption=Linear}
    alphaScaleBiasClamp {bias=1, scale=-1}
    boneToModify {boneName=ik_hand_l}
   }
   n7 = Leg IK!(ComponentPose:Component Space Pose Link Structure=@n6, Alpha:float=Get RightHandIKAlpha() -> (RightHandIKAlpha:float)) -> (Pose:Component Space Pose Link Structure) {
    is=title AnimGraphNode_LegIK
    alphaBoolBlend {blendOption=Linear}
    legsDefinition [{
      bEnableKneeTwistCorrection=false
      bEnableRotationLimit=false
      footBoneForwardAxis=Y
      hingeRotationAxis=Z
      minRotationAngle=15
      numBonesInLimb=2
      twistOffsetCurveName=None
      fKFootBone {boneName=hand_r}
      iKFootBone {boneName=ik_hand_r}
     }]
   }
   n8 = Leg IK(ComponentPose=@n7, Alpha=Get LeftHandIKAlpha() -> (LeftHandIKAlpha:float)) {
    alphaBoolBlend {blendOption=Linear}
    legsDefinition [{
      bEnableKneeTwistCorrection=false
      bEnableRotationLimit=false
      footBoneForwardAxis=Y
      hingeRotationAxis=Z
      minRotationAngle=15
      numBonesInLimb=2
      twistOffsetCurveName=None
      fKFootBone {boneName=hand_l}
      iKFootBone {boneName=ik_hand_l}
     }]
   }
   Output Pose(Result=Component To Local(ComponentPose=@n8)) {layerGroup=VehicleOverride}
   n9 = Get FerretVehicleInput() -> (FerretVehicleInput:Fort Anim Input Ferret Vehicle Structure)
   Save cached pose 'DriverAirPose'(Pose=Blendspace Player 'FerretVehicle_Driver_AirPoses_BS'(
     X=Break Fort Anim Input Ferret Vehicle!(FortAnimInput_FerretVehicle:"Fort Anim Input Ferret Vehicle Structure (by ref)"=@n9) -> (
      bIsUsingFerretVehicle:bool=~
      bIsDriver:bool=~
      bIsFrontPassenger:bool=~
      bIsBackLeftPassenger:bool=~
      bIsBackRightPassenger:bool=~
      bIsFrontPassengerAndLeaning:bool=~
      bIsBackPassengerAndLeaning:bool=~
      bIsDrifting:bool=~
      bIsBoosting:bool=~
      bIsReversing:bool=~
      bIsBraking:bool=~
      bIsMoving:bool=~
      bIsMovingForward:bool=~
      bIsLeaning:bool=~
      bIsLeaningOrBouncing:bool=~
      bIsBounceCrouching:bool=~
      bIsBounceCrouched:bool=~
      bIsBounceJumping:bool=~
      bIsBounceRecoiling:bool=~
      bIsSteeringRight:Byte
      bIsSteeringLeft:bool=~
      bIsShooting:bool=~
      bIsFerretPassengerRotating:bool=~
      RunForwardAlpha:float=~
      BounceCompression:float=~
      LeanPosition:Vector=~
      LeanPositionX:float=~
      LeanPositionY:float=~
      LeanPositionZ:float=~
      bAimFWD:bool=~
      bAimBWD:bool=~
      bAimLFT:bool=~
      bAimRGT:bool=~
      PawnToVehicleDeltaYawAngleDegrees:float=~
      AimCardDirDeadZoneAngleDegrees:float=~
      AimCardDirAngleOffsetDegrees:float=~
      AimFWDDeltaAngleDegrees:float=~
      AimBWDDeltaAngleDegrees:float=~
      AimLFTDeltaAngleDegrees:float=~
      AimRGTDeltaAngleDegrees:float=~
      SlopePitchDegreeAngle:float=~
      SlopeRollDegreeAngle:float=~
      SteerAngle:float=~
      SeatSwitchDirection:Vector=~
      bWasUp:bool=~
      PrevPlayerYAxis:Vector=~
      PrevPlayerZAxis:Vector=~
     ) {is=title K2Node_BreakStruct}.SteerAngle
     Y=Get ElevatorRotation() -> (ElevatorRotation:float)
    ) {blendSpace=&/FerretVehicle_Driver_AirPoses_BS}) {cacheName=DriverAirPose}
  ]}
 Locomotion {
  entry=MovementLoop
  path=VehicleFinalPoseOverride.AnimGraphNode_StateMachine_0.Locomotion
  ids=id
  nodes [
   AnimStateNode_0 = Ground() {is=title AnimStateNode}
   AnimStateNode_10 = RollE_To_MovementLoop() {is=title AnimStateNode}
   AnimStateNode_11 = RollW_To_MovementLoop() {is=title AnimStateNode}
   AnimStateNode_3 = MovementLoop() {is=title AnimStateNode}
   AnimStateNode_5 = BoostingEnd() {is=title AnimStateNode}
   AnimStateNode_6 = BoostingStart() {is=title AnimStateNode}
   AnimStateNode_7 = BoostingLoop() {is=title AnimStateNode}
   AnimStateNode_8 = Roll_E() {is=title AnimStateNode}
   AnimStateNode_9 = Roll_W() {is=title AnimStateNode}
   AnimStateTransitionNode_1 = AnimStateTransitionNode() {bAutomaticRuleBasedOnSequencePlayerInState=true, crossfadeDuration=0.4}
   AnimStateTransitionNode_10 = AnimStateTransitionNode() {bAutomaticRuleBasedOnSequencePlayerInState=true, crossfadeDuration=0.4}
   AnimStateTransitionNode_11 = AnimStateTransitionNode() {crossfadeDuration=0.1}
   AnimStateTransitionNode_12 = AnimStateTransitionNode() {crossfadeDuration=0.4}
   AnimStateTransitionNode_13 = AnimStateTransitionNode() {bAutomaticRuleBasedOnSequencePlayerInState=true}
   AnimStateTransitionNode_14 = AnimStateTransitionNode()
   AnimStateTransitionNode_15 = AnimStateTransitionNode() {bAutomaticRuleBasedOnSequencePlayerInState=true}
   AnimStateTransitionNode_2 = AnimStateTransitionNode()
   AnimStateTransitionNode_3 = AnimStateTransitionNode() {crossfadeDuration=0.4}
   AnimStateTransitionNode_4 = AnimStateTransitionNode()
   AnimStateTransitionNode_5 = AnimStateTransitionNode()
   AnimStateTransitionNode_7 = AnimStateTransitionNode() {crossfadeDuration=0.4}
  ]
  links [
   {
    dst=MovementLoop
    rule=VehicleFinalPoseOverride.AnimGraphNode_StateMachine_0.Locomotion.AnimStateTransitionNode_15.Transition
    src=BoostingEnd
    via=AnimStateTransitionNode_15
   }
   {
    dst=BoostingEnd
    rule=VehicleFinalPoseOverride.AnimGraphNode_StateMachine_0.Locomotion.AnimStateTransitionNode_14.AnimationTransitionGraph_1
    src=BoostingLoop
    via=AnimStateTransitionNode_14
   }
   {
    dst=BoostingLoop
    rule=VehicleFinalPoseOverride.AnimGraphNode_StateMachine_0.Locomotion.AnimStateTransitionNode_13.AnimationTransitionGraph_1
    src=BoostingStart
    via=AnimStateTransitionNode_13
   }
   {
    dst=MovementLoop
    rule=VehicleFinalPoseOverride.AnimGraphNode_StateMachine_0.Locomotion.AnimStateTransitionNode_7.Transition
    src=Ground
    via=AnimStateTransitionNode_7
   }
   {
    dst=BoostingStart
    rule=VehicleFinalPoseOverride.AnimGraphNode_StateMachine_0.Locomotion.AnimStateTransitionNode_11.Transition
    src=MovementLoop
    via=AnimStateTransitionNode_11
   }
   {
    dst=Ground
    rule=VehicleFinalPoseOverride.AnimGraphNode_StateMachine_0.Locomotion.AnimStateTransitionNode_2.Transition
    src=MovementLoop
    via=AnimStateTransitionNode_2
   }
   {
    dst=Roll_E
    rule=VehicleFinalPoseOverride.AnimGraphNode_StateMachine_0.Locomotion.AnimStateTransitionNode_4.Transition
    src=MovementLoop
    via=AnimStateTransitionNode_4
   }
   {
    dst=Roll_W
    rule=VehicleFinalPoseOverride.AnimGraphNode_StateMachine_0.Locomotion.AnimStateTransitionNode_5.Transition
    src=MovementLoop
    via=AnimStateTransitionNode_5
   }
   {
    dst=MovementLoop
    rule=VehicleFinalPoseOverride.AnimGraphNode_StateMachine_0.Locomotion.AnimStateTransitionNode_12.Transition
    src=RollE_To_MovementLoop
    via=AnimStateTransitionNode_12
   }
   {
    dst=MovementLoop
    rule=VehicleFinalPoseOverride.AnimGraphNode_StateMachine_0.Locomotion.AnimStateTransitionNode_3.Transition
    src=RollW_To_MovementLoop
    via=AnimStateTransitionNode_3
   }
   {
    dst=RollE_To_MovementLoop
    rule=VehicleFinalPoseOverride.AnimGraphNode_StateMachine_0.Locomotion.AnimStateTransitionNode_10.Transition
    src=Roll_E
    via=AnimStateTransitionNode_10
   }
   {
    dst=RollW_To_MovementLoop
    rule=VehicleFinalPoseOverride.AnimGraphNode_StateMachine_0.Locomotion.AnimStateTransitionNode_1.Transition
    src=Roll_W
    via=AnimStateTransitionNode_1
   }
  ]
 }
 Ground {
  path=VehicleFinalPoseOverride.AnimGraphNode_StateMachine_0.Locomotion.AnimStateNode_0.Ground
  nodes [Output Animation Pose(Result=Blendspace Player 'FerretVehicle_Driver_GroundPoses_BS'(X=Get GroundPoseBlendSpaceYaw() -> (GroundPoseBlendSpaceYaw:float), Y=Get GroundPoseBlendSpacePitch() -> (GroundPoseBlendSpacePitch:float)) {blendSpace=&/FerretVehicle_Driver_GroundPoses_BS})]
 }
 RollE_To_MovementLoop {
  path=VehicleFinalPoseOverride.AnimGraphNode_StateMachine_0.Locomotion.AnimStateNode_10.RollE_To_MovementLoop
  nodes [Output Animation Pose(Result=Use cached pose 'DriverAirPose'())]
 }
 RollW_To_MovementLoop {
  path=VehicleFinalPoseOverride.AnimGraphNode_StateMachine_0.Locomotion.AnimStateNode_11.RollW_To_MovementLoop
  nodes [Output Animation Pose(Result=Use cached pose 'DriverAirPose'())]
 }
 MovementLoop {
  path=VehicleFinalPoseOverride.AnimGraphNode_StateMachine_0.Locomotion.AnimStateNode_3.MovementLoop
  nodes [Output Animation Pose(Result=Use cached pose 'DriverAirPose'())]
 }
 BoostingEnd {
  path=VehicleFinalPoseOverride.AnimGraphNode_StateMachine_0.Locomotion.AnimStateNode_5.BoostingEnd
  nodes [Output Animation Pose(Result=Sequence Player 'FerretVehicle_Drive_S_Stop'() {bLoopAnimation=false, sequence=&/FerretVehicle_Drive_S_Stop})]
 }
 BoostingStart {
  path=VehicleFinalPoseOverride.AnimGraphNode_StateMachine_0.Locomotion.AnimStateNode_6.BoostingStart
  nodes [Output Animation Pose(Result=Sequence Player 'FerretVehicle_Drive_S_Start'() {bLoopAnimation=false, sequence=&/FerretVehicle_Drive_S_Start})]
 }
 BoostingLoop {
  path=VehicleFinalPoseOverride.AnimGraphNode_StateMachine_0.Locomotion.AnimStateNode_7.BoostingLoop
  nodes [Output Animation Pose(Result=Sequence Player 'FerretVehicle_Drive_S_CMM'() {sequence=&/FerretVehicle_Drive_S_CMM})]
 }
 Roll_E {
  path=VehicleFinalPoseOverride.AnimGraphNode_StateMachine_0.Locomotion.AnimStateNode_8.Roll_E
  nodes [
   n1 = Sequence Player 'FerretVehicle_Drive_E_Roll_Start'() {bLoopAnimation=false, playRate=0.6, sequence=&/FerretVehicle_Drive_E_Roll_Start}
   Output Animation Pose(Result=@n1)
  ]
 }
 Roll_W {
  path=VehicleFinalPoseOverride.AnimGraphNode_StateMachine_0.Locomotion.AnimStateNode_9.Roll_W
  nodes [
   n1 = Sequence Player 'FerretVehicle_Drive_W_Roll_Start'() {bLoopAnimation=false, playRate=0.6, sequence=&/FerretVehicle_Drive_W_Roll_Start}
   Output Animation Pose(Result=@n1)
  ]
 }
 Transition {path=VehicleFinalPoseOverride.AnimGraphNode_StateMachine_0.Locomotion.AnimStateTransitionNode_1.Transition, nodes [Result()]}
 Transition {path=VehicleFinalPoseOverride.AnimGraphNode_StateMachine_0.Locomotion.AnimStateTransitionNode_10.Transition, nodes [Result()]}
 Transition {
  path=VehicleFinalPoseOverride.AnimGraphNode_StateMachine_0.Locomotion.AnimStateTransitionNode_11.Transition
  nodes [Result(bCanEnterTransition=Get bIsBoosting() -> (bIsBoosting:bool))]
 }
 Transition {
  path=VehicleFinalPoseOverride.AnimGraphNode_StateMachine_0.Locomotion.AnimStateTransitionNode_12.Transition
  nodes [Result(bCanEnterTransition=NOT Boolean(A=Get bIsRollingRight() -> (bIsRollingRight:bool)))]
 }
 AnimationTransitionGraph_1 {
  path=VehicleFinalPoseOverride.AnimGraphNode_StateMachine_0.Locomotion.AnimStateTransitionNode_13.AnimationTransitionGraph_1
  nodes [Result()]
 }
 AnimationTransitionGraph_1 {
  path=VehicleFinalPoseOverride.AnimGraphNode_StateMachine_0.Locomotion.AnimStateTransitionNode_14.AnimationTransitionGraph_1
  nodes [Result(bCanEnterTransition=NOT Boolean(A=Get bIsBoosting()))]
 }
 Transition {path=VehicleFinalPoseOverride.AnimGraphNode_StateMachine_0.Locomotion.AnimStateTransitionNode_15.Transition, nodes [Result()]}
 Transition {
  path=VehicleFinalPoseOverride.AnimGraphNode_StateMachine_0.Locomotion.AnimStateTransitionNode_2.Transition
  nodes [Result(bCanEnterTransition=Get bOnGround() -> (bOnGround:bool))]
 }
 Transition {
  path=VehicleFinalPoseOverride.AnimGraphNode_StateMachine_0.Locomotion.AnimStateTransitionNode_3.Transition
  nodes [Result(bCanEnterTransition=NOT Boolean(A=Get bIsRollingLeft() -> (bIsRollingLeft:bool)))]
 }
 Transition {
  path=VehicleFinalPoseOverride.AnimGraphNode_StateMachine_0.Locomotion.AnimStateTransitionNode_4.Transition
  nodes [Result(bCanEnterTransition=Get bIsRollingRight())]
 }
 Transition {
  path=VehicleFinalPoseOverride.AnimGraphNode_StateMachine_0.Locomotion.AnimStateTransitionNode_5.Transition
  nodes [Result(bCanEnterTransition=Get bIsRollingLeft())]
 }
 Transition {
  path=VehicleFinalPoseOverride.AnimGraphNode_StateMachine_0.Locomotion.AnimStateTransitionNode_7.Transition
  nodes [Result(bCanEnterTransition=NOT Boolean(A=Get bOnGround()))]
 }
 FullBodyAdditive {
  entry=Idle
  path=VehicleFinalPoseOverride.AnimGraphNode_StateMachine_1.FullBodyAdditive
  ids=id
  nodes [
   AnimStateNode_0 = Idle()
   AnimStateNode_1 = Collision_N() {is=title AnimStateNode}
   AnimStateNode_2 = Collision_S() {is=title AnimStateNode}
   AnimStateTransitionNode_0 = AnimStateTransitionNode()
   AnimStateTransitionNode_1 = AnimStateTransitionNode() {bAutomaticRuleBasedOnSequencePlayerInState=true}
   AnimStateTransitionNode_2 = AnimStateTransitionNode() {crossfadeDuration=0.033}
   AnimStateTransitionNode_3 = AnimStateTransitionNode() {bAutomaticRuleBasedOnSequencePlayerInState=true}
  ]
  links [
   {
    dst=Idle
    rule=VehicleFinalPoseOverride.AnimGraphNode_StateMachine_1.FullBodyAdditive.AnimStateTransitionNode_3.Transition
    src=Collision_N
    via=AnimStateTransitionNode_3
   }
   {
    dst=Idle
    rule=VehicleFinalPoseOverride.AnimGraphNode_StateMachine_1.FullBodyAdditive.AnimStateTransitionNode_1.Transition
    src=Collision_S
    via=AnimStateTransitionNode_1
   }
   {
    dst=Collision_N
    rule=VehicleFinalPoseOverride.AnimGraphNode_StateMachine_1.FullBodyAdditive.AnimStateTransitionNode_2.Transition
    src=Idle
    via=AnimStateTransitionNode_2
   }
   {
    dst=Collision_S
    rule=VehicleFinalPoseOverride.AnimGraphNode_StateMachine_1.FullBodyAdditive.AnimStateTransitionNode_0.Transition
    src=Idle
    via=AnimStateTransitionNode_0
   }
  ]
 }
 Idle {
  path=VehicleFinalPoseOverride.AnimGraphNode_StateMachine_1.FullBodyAdditive.AnimStateNode_0.Idle
  nodes [
   n1 = "Sequence Player (additive) 'FerretVehicle_Driver_Idle'"() {sequence=&/FerretVehicle_Driver_Idle}
   n2 = "Sequence Player (additive) 'FerretVehicle_Driver_Idle_Fast'"() {sequence=&/FerretVehicle_Driver_Idle_Fast}
   n3 = Get bIsFerretRotating() -> (bIsFerretRotating:bool)
   Output Animation Pose(Result=Blend Poses by bool(
     BlendPose_0="Sequence Player (additive) 'FerretVehicle_Driver_Idle'"() {sequence=&FerretVehicle_Driver_Idle}
     BlendPose_1=Blend Poses by bool(BlendPose_0=@n2, BlendPose_1=@n1, BlendTime_0=0.4, BlendTime_1=0.4, bActiveValue=@n3) {blendType=Linear}
     bActiveValue=Get bOnGround()
    ) {blendType=Linear})
   Get bOnGround()
  ]
 }
 Collision_N {
  path=VehicleFinalPoseOverride.AnimGraphNode_StateMachine_1.FullBodyAdditive.AnimStateNode_1.Collision_N
  nodes [Output Animation Pose(Result="Sequence Player (additive) 'FerretVehicle_Driver_Collision_N'"() {bLoopAnimation=false, sequence=&/FerretVehicle_Driver_Collision_N})]
 }
 Collision_S {
  path=VehicleFinalPoseOverride.AnimGraphNode_StateMachine_1.FullBodyAdditive.AnimStateNode_2.Collision_S
  nodes [Output Animation Pose(Result="Sequence Player (additive) 'FerretVehicle_Driver_Collision_S'"() {bLoopAnimation=false, sequence=&/FerretVehicle_Driver_Collision_S})]
 }
 Transition {
  path=VehicleFinalPoseOverride.AnimGraphNode_StateMachine_1.FullBodyAdditive.AnimStateTransitionNode_0.Transition
  nodes [Result(bCanEnterTransition=Break Fort Anim Input Velocity Impact!(FortAnimInput_VelocityImpact:"Fort Anim Input Velocity Impact Structure (by ref)"=Get VelocityImpact() -> (VelocityImpact:Fort Anim Input Velocity Impact Structure)) -> (
     DeltaVelocityThreshold:Vector=~
     ImpactScale:Vector=~
     ImpactLimitX:Input Range Structure
     ImpactLimitY:Input Range Structure
     ImpactLimitZ:Input Range Structure
     SpringInterpolatorX:Float RK4Spring Interpolator Structure
     SpringInterpolatorY:Float RK4Spring Interpolator Structure
     SpringInterpolatorZ:Float RK4Spring Interpolator Structure
     bTestVelocity:bool=~
     TestVelocity:Vector=~
     bIsForwardImpact:bool=~
     bIsBackwardImpact:bool=~
     bIsLeftImpact:bool=~
     bIsRightImpact:bool=~
     bIsUpImpact:bool=~
     bIsDownImpact:bool=~
    ) {is=title K2Node_BreakStruct}.bIsBackwardImpact)]
 }
 Transition {path=VehicleFinalPoseOverride.AnimGraphNode_StateMachine_1.FullBodyAdditive.AnimStateTransitionNode_1.Transition, nodes [Result()]}
 Transition {
  path=VehicleFinalPoseOverride.AnimGraphNode_StateMachine_1.FullBodyAdditive.AnimStateTransitionNode_2.Transition
  nodes [Result(bCanEnterTransition=Break Fort Anim Input Velocity Impact(FortAnimInput_VelocityImpact=Get VelocityImpact()) -> (
     DeltaVelocityThreshold=~
     ImpactScale=~
     bTestVelocity=~
     TestVelocity=~
     bIsForwardImpact=~
     bIsBackwardImpact=~
     bIsLeftImpact=~
     bIsRightImpact=~
     bIsUpImpact=~
     bIsDownImpact=~
    ).bIsForwardImpact)]
 }
 Transition {path=VehicleFinalPoseOverride.AnimGraphNode_StateMachine_1.FullBodyAdditive.AnimStateTransitionNode_3.Transition, nodes [Result()]}
 Shooting {
  entry=Default
  path=VehicleFinalPoseOverride.AnimGraphNode_StateMachine_2.Shooting
  ids=id
  nodes [
   AnimStateNode_0 = Default() {is=title AnimStateNode}
   AnimStateNode_1 = Intro() {is=title AnimStateNode}
   AnimStateNode_2 = Loop() {is=title AnimStateNode}
   AnimStateNode_3 = Outro() {is=title AnimStateNode}
   AnimStateTransitionNode_2 = AnimStateTransitionNode() {crossfadeDuration=0}
   AnimStateTransitionNode_3 = AnimStateTransitionNode() {bAutomaticRuleBasedOnSequencePlayerInState=true, crossfadeDuration=0.1}
   AnimStateTransitionNode_6 = AnimStateTransitionNode() {crossfadeDuration=0.067}
   AnimStateTransitionNode_7 = AnimStateTransitionNode() {bAutomaticRuleBasedOnSequencePlayerInState=true, crossfadeDuration=0.1}
   AnimStateTransitionNode_8 = AnimStateTransitionNode() {crossfadeDuration=0}
  ]
  links [
   {
    dst=Intro
    rule=VehicleFinalPoseOverride.AnimGraphNode_StateMachine_2.Shooting.AnimStateTransitionNode_2.Transition
    src=Default
    via=AnimStateTransitionNode_2
   }
   {
    dst=Loop
    rule=VehicleFinalPoseOverride.AnimGraphNode_StateMachine_2.Shooting.AnimStateTransitionNode_3.Transition
    src=Intro
    via=AnimStateTransitionNode_3
   }
   {
    dst=Outro
    rule=VehicleFinalPoseOverride.AnimGraphNode_StateMachine_2.Shooting.AnimStateTransitionNode_6.Transition
    src=Loop
    via=AnimStateTransitionNode_6
   }
   {
    dst=Default
    rule=VehicleFinalPoseOverride.AnimGraphNode_StateMachine_2.Shooting.AnimStateTransitionNode_7.Transition
    src=Outro
    via=AnimStateTransitionNode_7
   }
   {
    dst=Intro
    rule=VehicleFinalPoseOverride.AnimGraphNode_StateMachine_2.Shooting.AnimStateTransitionNode_8.Transition
    src=Outro
    via=AnimStateTransitionNode_8
   }
  ]
 }
 Default {
  path=VehicleFinalPoseOverride.AnimGraphNode_StateMachine_2.Shooting.AnimStateNode_0.Default
  nodes [Output Animation Pose(Result=Additive Identity Pose())]
 }
 Intro {
  path=VehicleFinalPoseOverride.AnimGraphNode_StateMachine_2.Shooting.AnimStateNode_1.Intro
  nodes [Output Animation Pose(Result="Sequence Player (additive) 'FerretVehicle_Driver_Fire_Intro'"() {bLoopAnimation=false, sequence=&/FerretVehicle_Driver_Fire_Intro})]
 }
 Loop {
  path=VehicleFinalPoseOverride.AnimGraphNode_StateMachine_2.Shooting.AnimStateNode_2.Loop
  nodes [Output Animation Pose(Result="Sequence Player (additive) 'FerretVehicle_Driver_Fire_Loop'"() {sequence=&/FerretVehicle_Driver_Fire_Loop})]
 }
 Outro {
  path=VehicleFinalPoseOverride.AnimGraphNode_StateMachine_2.Shooting.AnimStateNode_3.Outro
  nodes [Output Animation Pose(Result="Sequence Player (additive) 'FerretVehicle_Driver_Fire_Outro'"() {bLoopAnimation=false, sequence=&/FerretVehicle_Driver_Fire_Outro})]
 }
 Transition {
  path=VehicleFinalPoseOverride.AnimGraphNode_StateMachine_2.Shooting.AnimStateTransitionNode_2.Transition
  nodes [Result(bCanEnterTransition=Get bIsFerretShooting() -> (bIsFerretShooting:bool))]
 }
 Transition {path=VehicleFinalPoseOverride.AnimGraphNode_StateMachine_2.Shooting.AnimStateTransitionNode_3.Transition, nodes [Result()]}
 Transition {
  path=VehicleFinalPoseOverride.AnimGraphNode_StateMachine_2.Shooting.AnimStateTransitionNode_6.Transition
  nodes [Result(bCanEnterTransition=NOT Boolean(A=Get bIsFerretShooting()))]
 }
 Transition {path=VehicleFinalPoseOverride.AnimGraphNode_StateMachine_2.Shooting.AnimStateTransitionNode_7.Transition, nodes [Result()]}
 Transition {
  path=VehicleFinalPoseOverride.AnimGraphNode_StateMachine_2.Shooting.AnimStateTransitionNode_8.Transition
  nodes [Result(bCanEnterTransition=Get bIsFerretShooting())]
 }
 VehicleFullBodyOverride {nodes [
   Input Pose() {name=InPoseFullBody}
   Output Pose() {layerGroup=VehicleOverride}
  ]}
 VehicleLowerBodyOverride {nodes [
   Input Pose() {name=InPoseLowerBodyDefault}
   Input Pose() {name=InPoseUpperBody}
   Output Pose() {layerGroup=VehicleOverride}
  ]}
 VehicleSplitBodyOverride {nodes [
   Input Pose() {name=InPoseSplitBody}
   Input Pose() {name=InPoseUpperAndLowerBody}
   Input Pose() {name=InPoseUpperAndLowerBodyRemoveRoot}
   Output Pose() {layerGroup=Vehicle Override}
  ]}
]
variables [
 HandLocalRotationOffsetLeft {type=Rotator}
 HandLocalRotationOffsetRight {type=Rotator}
 HandLocalTranslateOffsetLeft {type=Vector}
 HandLocalTranslateOffsetRight {type=Vector}
 JethroXV_Driver_Alpha {type=number}
]
