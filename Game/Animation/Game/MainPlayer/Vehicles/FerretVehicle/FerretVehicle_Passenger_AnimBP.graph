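# Animation Blueprint for passengers in the Ferret vehicle, providing vehicle override layers on the Fortnite player skeleton.
# Only the split-body layer has wired logic: an aim offset driven by vehicle slope roll and pitch leans the root-removed body pose.
# A bool blend from the Ferret input struct switches instantly between that result and the incoming split-body pose.
# The main graph, event graph and other override layers are unwired stubs.
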AnimBlueprint /Game/Animation/Game/MainPlayer/Vehicles/FerretVehicle/FerretVehicle_Passenger_AnimBP
target_skeleton=&Fortnite_M_Avg_Player_Skeleton
graphs [
 AnimGraph {nodes [Output Pose() {layerGroup=None}]}
 EventGraph {nodes [
   TryGetPawnOwner()
   Event BlueprintUpdateAnimation()
  ]}
 VehicleFinalPoseOverride {nodes [
   Input Pose() {name=InFinalPose}
   Output Pose() {layerGroup=VehicleOverride}
  ]}
 VehicleFullBodyOverride {nodes [
   Input Pose() {name=InPoseFullBody}
   Output Pose() {layerGroup=VehicleOverride}
  ]}
 VehicleLowerBodyOverride {nodes [
   Input Pose() {name=InPoseLowerBodyDefault}
   Input Pose() {name=InPoseUpperBody}
   Output Pose() {layerGroup=VehicleOverride}
  ]}
 VehicleSplitBodyOverride {nodes [
   Input Pose() {name=InPoseUpperAndLowerBody}
   n1 = Break Fort Anim Input Ferret Vehicle!(FortAnimInput_FerretVehicle:"Fort Anim Input Ferret Vehicle Structure (by ref)"=Get FerretVehicleInput() -> (FerretVehicleInput:Fort Anim Input Ferret Vehicle Structure)) -> (
    bIsUsingFerretVehicle:bool=~
    bIsDriver:bool=~
    bIsFrontPassenger:bool=~
    bIsBackLeftPassenger:bool=~
    bIsBackRightPassenger:bool=~
    bIsFrontPassengerAndLeaning:bool=~
    bIsBackPassengerAndLeaning:bool=~
    bIsDrifting:bool=~
    bIsBoosting:bool=~
    bIsReversing:bool=~
    bIsBraking:bool=~
    bIsMoving:bool=~
    bIsMovingForward:bool=~
    bIsLeaning:bool=~
    bIsLeaningOrBouncing:bool=~
    bIsBounceCrouching:bool=~
    bIsBounceCrouched:bool=~
    bIsBounceJumping:bool=~
    bIsBounceRecoiling:bool=~
    bIsSteeringRight:Byte
    bIsSteeringLeft:bool=~
    bIsShooting:bool=~
    bIsFerretPassengerRotating:bool=~
    RunForwardAlpha:float=~
    BounceCompression:float=~
    LeanPosition:Vector=~
    LeanPositionX:float=~
    LeanPositionY:float=~
    LeanPositionZ:float=~
    bAimFWD:bool=~
    bAimBWD:bool=~
    bAimLFT:bool=~
    bAimRGT:bool=~
    PawnToVehicleDeltaYawAngleDegrees:float=~
    AimCardDirDeadZoneAngleDegrees:float=~
    AimCardDirAngleOffsetDegrees:float=~
    AimFWDDeltaAngleDegrees:float=~
    AimBWDDeltaAngleDegrees:float=~
    AimLFTDeltaAngleDegrees:float=~
    AimRGTDeltaAngleDegrees:float=~
    SlopePitchDegreeAngle:float=~
    SlopeRollDegreeAngle:float=~
    SteerAngle:float=~
    SeatSwitchDirection:Vector=~
    bWasUp:bool=~
    PrevPlayerYAxis:Vector=~
    PrevPlayerZAxis:Vector=~
   ) {is=title K2Node_BreakStruct}
   n2 = AimOffset Player 'FerretVehicle_PassengerSlope_AO'!(
    X:float=@n1.SlopeRollDegreeAngle
    Y:float=@n1.SlopePitchDegreeAngle
    BasePose:Pose Link Structure=Input Pose() {name=InPoseUpperAndLowerBodyRemoveRoot}
    Alpha:float=1
   ) -> (Pose:Pose Link Structure) {blendSpace=/Game/Animation/Game/MainPlayer/Vehicles/FerretVehicle/AO/FerretVehicle_PassengerSlope_AO, alphaBoolBlend {blendOption=Linear}}
   Output Pose(Result=Blend Poses by bool(BlendPose_0=@n2, BlendPose_1=Input Pose() {name=InPoseSplitBody}, BlendTime_0=0, BlendTime_1=0, bActiveValue=@n1) {blendType=Linear}) {layerGroup=Vehicle Override}
  ]}
]
variables []
